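; Fortnite Ascender zipline that players ride up by handle or slide down by cable.
; Parent-class events drive handle and motor VFX, parameterized audio, hostile-only sound indicators, weapon targeting blocks and timer-gated whoosh.
; Interaction overrides pick handle or cable by distance and filter by gameplay tags and DBNO state.
; Cable tiling and timestamp-driven wobble use material parameters; the actor can die with its linked actor or be hotfix-disabled.
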(event Custom|OnAscenderSetupComplete_Event
  (CallFunction|SetCableTilingbySplineLength))

(event EventBeginPlay
  (Class|FortAsyncActionCurrentPlaylistReady|PlaylistDataReadyAsync
    (:then
      (EventDispatchers|AssignOnAscenderSetupComplete (delegate Custom|OnAscenderSetupComplete_Event))
      (EventDispatchers|BindEventtoOnLinkedActorDestroyed (Variables|ActorLinking|GetLinkToActorComponent) (EventDispatchers|CreateEvent :!event Custom|HandleLinkedActorDestroyed)))
    (:OnReady
      (CallFunction|HotfixEnabled))))

(event EventHandlePlayerStartedUsingHandle
  (params (Player :type Fort%20Player%20Pawn%20Object%20Reference))
  (bind _returnvalue (CallFunction|GetPawnUsingHandle))
  (bind _handle (Variables|Default|GetHandle))
  (Utilities|FlowControl|Sequence
    (:then_0
      ;[0.1897 1 0] Store PlayerPawn and the PlayerState that is using the Handle Component.
      (Utilities|IsValid _returnvalue
        (:Is%20Valid
          (bind _output_get (Variables|Default|SetPlayerPawn _returnvalue))
          (Variables|Default|SetPlayerStateUsingHandleComponent (Variables|/Script/Engine.Pawn|GetPlayerState _output_get))
          (Utilities|IsValid (AudioParameter|GetAudioParameterComponent _output_get :!class /Script/FortniteGame.FortPawn)
            (:Is%20Valid
              (bind _output_get_2 (Variables|Default|SetParameterComponent (AudioParameter|GetAudioParameterComponent _output_get :!class /Script/FortniteGame.FortPawn)))
              (Variables|Sound|SetAudioParameters (Parameters|GetParameters _output_get_2)))))
        (:Is%20Not%20Valid
          (Variables|Default|SetPlayerStateUsingHandleComponent 0))))
    (:then_1
      (CallFunction|ActivateCableWobble)
      (bind _returnvalue_2 (Audio|SpawnSoundAttached (Variables|Sound|GetStartHandleSound) _handle))
      (bind _output_get_3 (Variables|Default|SetHandleStart _returnvalue_2))
      (if (not (Pawn|IsLocallyViewed Player))
        (Fortnite|ShowSoundIndicatorLocalforTeam :Location (Transformation|GetWorldLocation _handle) :MaxAudibleDistance 0.0 :IndicatorType "Generic" :Team 0 :Affiliations (Utilities|Array|MakeArray "Hostile") :Tint "(R=0.000000,G=0.000000,B=0.000000,A=0.000000)" :OverrideIcon "/Ascender/Gameplay/Ascender/Icons/AV_Ascender_32.AV_Ascender_32" :SoundBase (Variables|/Script/Engine.AudioComponent|GetSound _returnvalue_2) :AttenuationMaxRangeMultiplier 1.0 :AttenuationIntensityMultiplier 1.0))
      (Utilities|IsValid _returnvalue_2
        (:Is%20Valid
          (Audio|Parameter|SetParameters _output_get_3 (Variables|Sound|GetAudioParameters)))
        (:Is%20Not%20Valid
          _))
      (bind _handleascendingsfx (Variables|Default|GetHandleAscendingSFX)
        (:then
          (Audio|Components|Audio|FadeIn _handleascendingsfx 0.2 1.0 0.0 "Linear" :!class /Script/Engine.AudioComponent))
        (:else
          _))
      (Niagara|SetNiagaraSystemAsset (Variables|Default|GetHandleVFX) (Variables|Default|GetAscendingHandleVFX) true)
      (Components|Activation|Activate (Variables|Default|GetHandleVFX) true)
      (Components|Activation|Deactivate (Variables|Default|GetHandleDescendingSFX)))))

(event EventHandlePlayerStoppedUsingHandle
  (params (Player :type Fort%20Player%20Pawn%20Object%20Reference))
  (bind _handle (Variables|Default|GetHandle))
  (bind _returnvalue (Transformation|GetWorldLocation _handle))
  (bind _stophandlesound (Variables|Sound|GetStopHandleSound))
  (bind _handlestart (Variables|Default|GetHandleStart)
    (:then
      (Audio|Components|Audio|FadeOut _handlestart 0.2 0.0 "Linear" :!class /Script/Engine.AudioComponent))
    (:else
      _))
  (CallFunction|DeactivateCableWobble)
  (bind _returnvalue_2 (Audio|SpawnSoundatLocation _stophandlesound _returnvalue))
  (if (not (Pawn|IsLocallyViewed Player))
    (Fortnite|ShowSoundIndicatorLocalforTeam :Location _returnvalue :MaxAudibleDistance 0.0 :IndicatorType "Generic" :Team 0 :Affiliations (Utilities|Array|MakeArray "Hostile") :Tint "(R=0.000000,G=0.000000,B=0.000000,A=0.000000)" :OverrideIcon "/Ascender/Gameplay/Ascender/Icons/AV_Ascender_32.AV_Ascender_32" :SoundBase _stophandlesound :AttenuationMaxRangeMultiplier 1.0 :AttenuationIntensityMultiplier 1.0))
  (Audio|Parameter|SetParameters _returnvalue_2 (Variables|Sound|GetAudioParameters))
  (Audio|Components|Audio|FadeOut (Variables|Default|GetHandleAscendingSFX) 0.1 0.0 "Linear" :!class /Script/Engine.AudioComponent)
  (if (Variables|Sound|GetShouldPlayWhoosh)
    (bind _playerpawn (Variables|Default|GetPlayerPawn)
      (:then
        (bind _returnvalue_3 (Audio|SpawnSoundAttached (Variables|Sound|GetAscendWhooshSound) (Variables|/Script/Engine.Character|GetMesh _playerpawn) "head"))
        (Audio|Parameter|SetParameters _returnvalue_3 (Variables|Sound|GetAudioParameters)))
      (:else))))

(event EventHandleStartedLoweringCable
  (Components|Activation|Activate (Variables|Default|GetPoleMotorVFX) true))

(event EventHandleStartedLoweringHandle
  (bind _handledescendingsfx (Variables|Default|GetHandleDescendingSFX))
  (Niagara|SetNiagaraSystemAsset (Variables|Default|GetHandleVFX) (Variables|Default|GetDescendingHandleVFX) true)
  (Components|Activation|Activate (Variables|Default|GetHandleVFX) true)
  (Audio|Components|Audio|FadeIn _handledescendingsfx 0.2 1.0 0.0 "Linear" :!class /Script/Engine.AudioComponent)
  (Audio|Parameter|SetParameters _handledescendingsfx (Variables|Sound|GetAudioParameters)))

(event EventHandleStoppedLoweringCable
  (Components|Activation|Deactivate (Variables|Default|GetPoleMotorVFX))
  (Audio|Components|Audio|FadeOut (Variables|Default|GetHandleDescendingSFX) 0.2 0.0 "Linear" :!class /Script/Engine.AudioComponent))

(event EventHandleStoppedLoweringHandle
  (Utilities|IsValid (CallFunction|GetPawnUsingHandle)
    (:Is%20Valid
      _)
    (:Is%20Not%20Valid
      (bind _returnvalue (Audio|SpawnSoundAttached (Variables|Sound|GetStopPoleMotorSoundSkid) (Variables|Default|GetHandle)))
      (Audio|Parameter|SetParameters _returnvalue (Variables|Sound|GetAudioParameters))
      (Components|Activation|Deactivate (Variables|Default|GetHandleVFX))))
  (Audio|Components|Audio|FadeOut (Variables|Default|GetHandleDescendingSFX) 0.2 0.0 "Linear" :!class /Script/Engine.AudioComponent))

(event EventHandleUpdatedLoweringCable
  (CallFunction|AttachCaptoEndOfSplineMesh))

(event EventOnZipliningStarted
  (params (InteractingPawn :type Fort%20Player%20Pawn%20Object%20Reference) (InteractComponent :type Primitive%20Component%20Object%20Reference))
  ;[1 1 1] Handle Zipline Use Started
  (|CallParentFunction InteractingPawn InteractComponent)
  (Utilities|FlowControl|Sequence
    (:then_0
      ;[1 1 1] Handle Player Started Sliding Down
      (CallFunction|PlayerDescendingStarted InteractingPawn))
    (:then_1
      ;[1 1 1] Player Started Using Ascender (Both Using Handle And Sliding Down)
      (Collision|AddIgnoredActor InteractingPawn self 9999.0))
    (:then_2
      (bind _returnvalue (Utilities|Time|SetTimerbyFunctionName 0 "ShouldPlayWhooshTimer" 1.0))
      (Variables|Sound|SetWhooshHandle _returnvalue))
    (:then_3
      ;[1 1 1] Block Weapon Canting Toggle
      (bind _playerpawn (Variables|Default|GetPlayerPawn)
        (:then
          (bind _asfort_weapon_ranged (Utilities|Casting|CastToFortWeaponRanged (Weapon|GetCurrentWeapon _playerpawn)))
          (Targeting|SetThirdPersonTargetingModeAllowed _asfort_weapon_ranged false false)
          (FortWeapon|SetTargeting _asfort_weapon_ranged false) ; Make the player exit targeting because\n1) players don't need it on the ascender because the can't shoot and\n2) it has camera and UI bugs which on the ascender.
        )
        (:else)))))

(event Custom|ShouldPlayWhooshTimer
  (Variables|Sound|SetShouldPlayWhoosh true))

(event EventOnZipliningStopped
  (params (InteractingPawn :type Fort%20Player%20Pawn%20Object%20Reference) (InteractComponent :type Primitive%20Component%20Object%20Reference) (ZiplineUsageDuration :type Float%20%28single-precision%29))
  ;[1 1 1] Handle Zipline Use Stopped
  (|CallParentFunction InteractingPawn InteractComponent ZiplineUsageDuration)
  (Utilities|FlowControl|Sequence
    (:then_0
      ;[1 1 1] Allow Weapon Canting Toggle
      (bind _playerpawn (Variables|Default|GetPlayerPawn)
        (:then
          (bind _asfort_weapon_ranged (Utilities|Casting|CastToFortWeaponRanged (Weapon|GetCurrentWeapon _playerpawn)))
          (Targeting|SetThirdPersonTargetingModeAllowed _asfort_weapon_ranged true true))
        (:else)))
    (:then_1
      ;[1 1 1] Player Stopped Using Ascender (Both Using Handle And Sliding Down)
      (Collision|AddIgnoredActor InteractingPawn self 0.5))
    (:then_2
      (Variables|Sound|SetShouldPlayWhoosh false)
      (Utilities|Time|ClearandInvalidateTimerbyHandle (Variables|Sound|GetWhooshHandle)))))

(event Custom|PlayerDescendingStarted
  (params (Sliding%20Player :type Fort%20Player%20Pawn%20Object%20Reference))
  (bind _returnvalue (AudioParameter|GetAudioParameterComponent Sliding%20Player :!class /Script/FortniteGame.FortPawn))
  (Utilities|IsValid Sliding%20Player
    (:Is%20Valid
      (bind _returnvalue_2 (Audio|SpawnSoundAttached (Variables|Sound|GetPlayerGrabBeginDecentSound) (Variables|/Script/Engine.Character|GetMesh Sliding%20Player) "head"))
      (Utilities|IsValid _returnvalue
        (:Is%20Valid
          (Audio|Parameter|SetParameters _returnvalue_2 (Parameters|GetParameters _returnvalue)))))))

(event Custom|HandleLinkedActorDestroyed
  (params (DamageCauser :type Actor%20Object%20Reference))
  ;[1 1 1] Die when foundation has been destroyed
  (if (ScalableFloat|GetValueasBool (Variables|Zipline|GetDestructionEnabled))
    (GameplayCue|ExecuteGameplayCueOnActor(Burst) self (Variables|Default|GetLinkToDestroyedGC) (Ability|GameplayCue|MakeGameplayCueParameters :NormalizedMagnitude 0.0 :RawMagnitude 1e+02 :Location (+ (Transformation|GetWorldLocation (Variables|Default|GetTop)) 0) :Normal "0,0,0" :SourceObject (Variables|Default|GetTop) :GameplayEffectLevel 1 :AbilityLevel 1 :bReplicateLocationWhenUsingMinimalRepProxy false))
    (Gameplay|Die :DamageCauser DamageCauser)
    (Variables|Default|SetBeDestroy true)))

(fn ActivateCableWobble
  (Networking|CanExecuteCosmeticEvents
    (:True
      (bind _output_get (Variables|Default|SetLastWobbleActivationTime (Utilities|Time|GetGameTimeinSeconds)))
      (Rendering|Material|SetScalarParameterValueonMaterials (Variables|FortAscenderZipline|GetSplineMesh) "TimeStampA" _output_get)
      (Rendering|Material|SetScalarParameterValueonMaterials (Variables|FortAscenderZipline|GetSplineMesh) "TimeStampB" (+ (Variables|Default|GetLastWobbleActivationTime) (Variables|Default|GetMaxWobbleAnimationLength))))))

(fn ApplyStructureDamage
  (params (BuildingActor :type Building%20Static%20Mesh%20Object%20Reference) (DamageSource :type Actor%20Object%20Reference))
  (Fortnite|ApplyGameplayEffectsToActor BuildingActor (Utilities|Array|MakeArray (Variables|Default|GetStructureDamageGE)) (Transformation|GetActorLocation DamageSource) false 0 (Ability|GetAbilitySystemComponent DamageSource)))

(fn AreExtraDBNOInteractionsEnabled
  (params (Pawn :type Fort%20Pawn%20Object%20Reference))
  (bind _returnvalue (Actor|GetComponentbyClass Pawn "/Script/FortniteGame.DBNOBehaviorOverrideComponent"))
  (Utilities|IsValid _returnvalue
    (:Is%20Valid
      (return (DBNO|AreExtraDBNOInteractionsEnabled _returnvalue)))
    (:Is%20Not%20Valid
      (return false))))

(fn AttachCapToEndOfSplineMesh
  (bind _ascender_end_cap (Variables|Default|GetAscender_End_Cap))
  (if (Math|Boolean|ANDBoolean (Utilities|IsValid (Variables|FortAthenaSplineZipline|GetZiplineSpline) :!class /Script/Engine.KismetSystemLibrary) (Utilities|IsValid (Variables|FortAscenderZipline|GetSplineMesh) :!class /Script/Engine.KismetSystemLibrary) (Utilities|IsValid (Variables|Default|GetAscender_End_Cap) :!class /Script/Engine.KismetSystemLibrary))
    (Transformation|SetWorldLocation _ascender_end_cap (+ (Variables|FortAscenderZipline|GetCurrentSplineEndLocation) (Math|Vector|MakeVector 0.0 0.0 (Variables|FortAscenderZipline|GetCableOffsetFromSplineEnd))) false false)
    (Transformation|SetRelativeScale3D _ascender_end_cap (Variables|Default|GetEndCapRelativeScale))))

(fn AttemptOverrideZiplineSocketOffset
  (params (InteractingPawn :type Fort%20Player%20Pawn%20Object%20Reference) (InteractComponent :type Primitive%20Component%20Object%20Reference) (BaseSocketOffset :type Vector%20%28by%20ref%29 :default "0, 0, 0"))
  (return (+ BaseSocketOffset (select (== InteractComponent (CallFunction|GetHandleComponent)) (Variables|Default|GetHandleAnimOffset) (Variables|Default|GetSlideAnimOffset))) true))

(fn BlueprintCanInteract
  (params (InteractingPawn :type Fort%20Pawn%20Object%20Reference) (InteractionBeingAttempted :type EInteractionBeingAttempted%20Enum :default "FirstInteraction") (InteractionType :type TInteractionType%20Enum :default "IT_NoInteraction"))
  (bind _returnvalue (Ability|GetAbilitySystemComponent InteractingPawn))
  (if (Utilities|IsValid InteractingPawn :!class /Script/Engine.KismetSystemLibrary)
    (if (Networking|HasAuthority InteractingPawn)
      (return true)
      (elif (and (not (GameplayTags|HasMatchingGameplayTag _returnvalue "(TagName=\"Abilities.Generic.Athena.Zipline\")")) (not (GameplayTags|HasAnyMatchingGameplayTags _returnvalue (LocalVariables|GetBlockInteractionTags))))
        (if (not (GameplayTags|HasAnyMatchingGameplayTags _returnvalue (LocalVariables|GetBlockInteractionTags_DBNO)))
          (return true)
          (else
            (bind _asfort_player_pawn (Utilities|Casting|CastToFortPlayerPawn InteractingPawn)
              (:then
                (if (Default|AreExtraDBNOInteractionsEnabled _asfort_player_pawn)
                  (return true)
                  (else
                    _)))
              (:CastFailed
                _))
            (return false)))
        (else
          (return false))))
    (else
      (return true))))

(fn BlueprintGetInteractionString
  (params (InteractingPawn :type Fort%20Pawn%20Object%20Reference) (InteractionBeingAttempted :type EInteractionBeingAttempted%20Enum :default "FirstInteraction"))
  (return (select (== (Variables|FortAscenderZipline|GetCurrentInteractComponent) (Variables|Default|GetHandle)) "NSLOCTEXT(\"[64C3E4A04544D7500AF36B939D97E520]\", \"04B09175432EBAF98E1223ABA5683F4B\", \"Ride Up\")" "NSLOCTEXT(\"[64C3E4A04544D7500AF36B939D97E520]\", \"2FE2B1B4437CBACAD1EFCC8DC1FCFC85\", \"Ride Down\")")))

(fn CalculateLaunchVelocity
  (params (PlayerPawn :type Actor%20Object%20Reference))
  (return (Utilities|Operators|Add (* (Transformation|GetActorForwardVector PlayerPawn) (* (ScalableFloat|GetValueatLevel (Variables|FortAscenderZipline|GetDescendJumpedOffHorizontalLaunchSpeed)) -1.0)) (Math|Vector|MakeVector 0.0 0.0 (ScalableFloat|GetValueatLevel (Variables|FortAscenderZipline|GetDescendJumpedOffVerticalLaunchSpeed))))))

(fn CanBeginZiplining
  (params (InteractingPawn :type Fort%20Player%20Pawn%20Object%20Reference) (InteractComponent :type Primitive%20Component%20Object%20Reference))
  ;[1 1 1] Trend Concentrate Hook
  (if (Default|AreExtraDBNOInteractionsEnabled InteractingPawn)
    (return false)
    (elif true
      (return (not (or (Health|IsDBNO InteractingPawn) (Utilities|IsValid (Class|FortPlayerPawn|GetDBNOHoistee InteractingPawn) :!class /Script/Engine.KismetSystemLibrary))))
      (else
        (return false)))))

(fn DeactivateCableWobble
  (Networking|CanExecuteCosmeticEvents
    (:True
      (if (not (> (Utilities|Time|GetGameTimeinSeconds) (+ (Variables|Default|GetLastWobbleActivationTime) (Variables|Default|GetMaxWobbleAnimationLength))))
        (Rendering|Material|SetScalarParameterValueonMaterials (Variables|FortAscenderZipline|GetSplineMesh) "TimeStampB" (Utilities|Time|GetGameTimeinSeconds))))))

(fn GetAffiliationsToShowFor
  (return (Utilities|Array|MakeArray "Neutral" "Friendly" "Hostile")))

(fn GetBoundsComponents
  (params (SelfActor :type Actor%20Object%20Reference))
  ;[1 1 1] Returning an empty array so that the interaction system uses the result of GetInteracComponentOverride
  (return 0))

(fn GetDynamicDestructionResourceType
  (return "Metal"))

(fn GetDynamicDestructionStaticMeshComponents
  (return (Utilities|Array|MakeArray (Variables|Default|GetTop) (Variables|Default|GetHandle)) (Utilities|Array|MakeArray (Variables|Default|GetAscender_End_Cap)) 0))

(fn GetHandleComponent
  (return (Variables|Default|GetHandle)))

(fn GetInteractComponentOverride
  (params (InteractingPawn :type Fort%20Player%20Pawn%20Object%20Reference) (InteractComponent :type Primitive%20Component%20Object%20Reference))
  (Utilities|IsValid InteractingPawn
    (:Is%20Valid
      (return (select (<= (Math|Vector|DistanceSquared(Vector) (Transformation|GetWorldLocation (Variables|/Script/Engine.Character|GetCapsuleComponent InteractingPawn)) (Transformation|GetWorldLocation (Variables|Default|GetHandleInteractVolume))) (Math|Float|Square (+ (Variables|/Script/Engine.SphereComponent|GetSphereRadius (Variables|Default|GetHandleInteractVolume)) 5e+01))) (Variables|Default|GetHandle) InteractComponent)))
    (:Is%20Not%20Valid
      (return InteractComponent))))

(fn GetLastSplinePointIndex
  (return (- (Spline|GetNumberOfSplinePoints (Variables|FortAthenaSplineZipline|GetZiplineSpline)) 1)))

(fn GetTopComponent
  (return (Variables|Default|GetTop)))

(fn HotfixEnabled
  (if (not (ScalableFloat|GetValueasBool (Variables|Default|GetHotfix)))
    (Actor|SetLifeSpan 0.01)))

(fn LocalOnFailedInteract
  (params (InteractingPawn :type Fort%20Player%20Pawn%20Object%20Reference)))

(fn OnRep_BeDestroy
  (if (Variables|Default|GetBeDestroy)
    (Development|PrintString "Visibility Off" true true "(R=1.000000,G=0.113467,B=0.003805,A=1.000000)")
    (Rendering|SetVisibility (Variables|Default|GetAscender_End_Cap) false true)))

(fn PawnIsInHandleRange
  (params (Pawn :type Fort%20Pawn%20Object%20Reference))
  (return (Utilities|Array|ContainsItem (Collision|GetOverlappingActors (Variables|Default|GetHandleInteractVolume) "/Script/FortniteGame.FortPawn" :!class /Script/Engine.PrimitiveComponent) Pawn)))

(fn SetCableTilingBySplineLength
  (Rendering|Material|SetScalarParameterValueonMaterials (Variables|FortAscenderZipline|GetSplineMesh) "Tile_V" (Math|Conversions|ToFloat(Integer) (Math|Float|Truncate (/ (Spline|GetSplineLength (Variables|FortAthenaSplineZipline|GetZiplineSpline)) (Variables|Default|GetTilingDivisor))))))

(fn ShouldShowSoundIndicator
  (params (PlayerController :type Fort%20Player%20Controller%20Object%20Reference))
  (Utilities|FlowControl|Sequence
    (:then_0
      ;[0.8249 0 1] Only show sound indicator if the player using the handle is hostile to the visualizing player.
      (Utilities|IsValid PlayerController
        (:Is%20Valid
          (return (Utilities|Enum|Equal(Enum) (Team|GetTeamAffiliation (Variables|/Script/Engine.Controller|GetPlayerState PlayerController) (Variables|Default|GetPlayerStateUsingHandleComponent)) "Hostile")))))
    (:then_1
      ;[1 0.474 0.474] In case of evaluation failure, return false.
      (return false))))

(fn ShouldUseDynamicDestructionMaterial
  (return true))

(fn ShouldUseDynamicDestructionNiagaraSystem
  (return true))

(fn ConstructionScript
  (bind _splinemesh (Variables|FortAscenderZipline|GetSplineMesh))
  (Ascender|SetupAscender true)
  (Networking|CanExecuteCosmeticEvents
    (:True
      (Utilities|FlowControl|Sequence
        (:then_0
          (bind _splinemeshmaterialoverride (Variables|Default|GetSplineMeshMaterialOverride)
            (:then
              (Rendering|Material|SetMaterial _splinemesh 0 _splinemeshmaterialoverride))
            (:else)))
        (:then_1
          (bind _splinethickenmeshmaterialoverride (Variables|Default|GetSplineThickenMeshMaterialOverride)
            (:then
              (Rendering|Material|SetMaterial _splinemesh 1 _splinethickenmeshmaterialoverride))
            (:else)))
        (:then_2
          (CallFunction|SetCableTilingbySplineLength))
        (:then_3
          (CallFunction|AttachCaptoEndOfSplineMesh))))))
