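; Child Blueprint variant of the Athena zipline Ascender that adds an adjustable extending top arm.
; Its construction script keeps the arm's relative X and Y, sets Z from the extension length, then runs the parent construction.
; GetTopComponent returns the arm so the cable attaches to it instead of the base mesh.
; The event graph holds only empty or disabled stubs.
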(event EventBeginPlay
  (|CallParentFunction) ; This node is disabled and will not be called.\nDrag off pins to build functionality.
)

(event Collision|EventActorBeginOverlap
  (params (OtherActor :type Actor%20Object%20Reference)))

(event EventTick
  (params (DeltaSeconds :type Float%20%28single-precision%29)))

(fn GetTopComponent
  ;[1 1 1] In this case we want the cable to be attached to the extending arm, not the base mesh which is attached to a wall for example.
  (return (Variables|Default|GetTopArm)))

(fn ConstructionScript
  (bind (_relativelocation_x _relativelocation_y) (Math|Vector|BreakVector (Variables|/Script/Engine.SceneComponent|GetRelativeLocation_X (Variables|Default|GetTopArm))))
  (Transformation|SetRelativeLocation (Variables|Default|GetTopArm) (Math|Vector|MakeVector _relativelocation_x _relativelocation_y (Variables|Default|GetTopArmExtensionLength)) false false)
  (|CallParentFunction))
