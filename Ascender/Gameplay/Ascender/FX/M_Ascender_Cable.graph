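# Masked clear-coat metallic cable material for the Ascender, usable on spline meshes and Niagara mesh particles, with tiled zipline cable textures.
# World position offset sums optional SplineThicken, a sine ripple based on player distance with time envelope and end falloff, and a sphere-masked spiral wiggle with optional return wiggle.
# Dynamic parameters scale the wiggle and drive a noise-based erosion along V that dissolves the cable.
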Material /Ascender/Gameplay/Ascender/FX/M_Ascender_Cable
graph {
 properties {
  bUsedWithNiagaraMeshParticles=true
  bUsedWithSplineMeshes=true
  blendMode=BLEND_Masked
  pixelDepthOffsetMode=PDOM_Legacy
  previewMesh=/Game/Effects/Fort_Effects/Mesh/Rope/SM_Grapple_Rope
  shadingModel=MSM_ClearCoat
  usedShadingModels=Clear Coat
 }
 nodes [
  n1 = ScalarParameter() {defaultValue=-10, parameterName=Width}
  n2 = MaterialFunctionCall(Float2=TextureCoordinate()) {materialFunction=&BreakOutFloat2Components}
  n3 = MaterialFunctionCall(WidthBase=@n1, WidthTip=@n1, UVs For NormalShading=AppendVector(A=OneMinus(None=@n2.R), B=@n2.G)) {materialFunction=&SplineThicken}
  n4 = StaticSwitchParameter(True=@n3.WorldPositionOffset, False=Constant3Vector()) {parameterName=IsSplineThickening}
  n5 = VectorParameter() {parameterName=PlayerLocation, defaultValue=<rgba -6620 -8661 2870 1>}
  n6 = CollectionParameter() {collection=&FortniteMaterialParameters, parameterName=PlayerPosition}
  n7 = ComponentMask(None=@n6) {b=true, g=true, r=true}
  n8 = Distance(
   A=Add(A=@n4, B=WorldPosition().XYZ)
   B=Add(A=StaticSwitchParameter(True=@n5.RGB, False=@n7) {parameterName=Debug}, B=VectorParameter() {parameterName=PlayerPositionBias}.RGB)
  )
  n9 = Add(
   A=Multiply(A=Time(), B=ScalarParameter() {defaultValue=-2, parameterName=Speed})
   B=Divide(A=@n8, B=VectorParameter() {parameterName=Frequency, defaultValue=<rgba 1134 1134 1134 0>}.RGB)
  )
  n10 = Divide(A=@n8, B=ScalarParameter() {defaultValue=8000, parameterName=DistanceFalloff})
  n11 = Multiply(A=VectorParameter() {parameterName=Amplitude, defaultValue=<rgba 0 0 15 0>}.RGB, B=OneMinus(None=Saturate(None=@n10)))
  n12 = ScalarParameter() {defaultValue=15.9454, parameterName=DistanceFalloffNearBias}
  n13 = Saturate(None=Divide(A=Subtract(A=@n8, B=@n12), B=ScalarParameter() {defaultValue=201.1301, parameterName=DistanceFalloffNear}))
  n14 = MaterialFunctionCall(
   TimestampA=ScalarParameter() {parameterName=TimestampA}
   TransitionLength_In=ScalarParameter() {defaultValue=0.2, parameterName=TransitionLength_In}
   TimestampB=ScalarParameter() {defaultValue=1, parameterName=TimestampB}
   TransitionLength_Out=ScalarParameter() {defaultValue=0.5, parameterName=TransitionLength_Out}
   Transition In Only?=StaticBool()
  ) {materialFunction=&MF_Timestamp_InOut_01}
  n15 = Subtract(A=Time(), B=ScalarParameter() {parameterName=TimeStamp})
  n16 = ScalarParameter() {defaultValue=1, parameterName=DurationMultiplier}
  n17 = Sine(None=Saturate(None=Multiply(A=@n15, B=@n16)))
  n18 = Multiply(
   A=Multiply(A=Sine(None=@n9), B=Multiply(A=@n11, B=@n13))
   B=StaticSwitchParameter(
    True=MaterialFunctionCall() {materialFunction=&DebugTimeSine}.Result
    False=StaticSwitchParameter(True=@n14.Linear, False=@n17) {parameterName=Use_In/Out_Timestamps}
   ) {parameterName=Debug}
  )
  n19 = ComponentMask(None=TextureCoordinate()) {g=true}
  n20 = LinearInterpolate(B=@n19, Alpha=ScalarParameter() {defaultValue=1, parameterName=UseStart}) {constA=1}
  n21 = LinearInterpolate(B=OneMinus(None=@n19), Alpha=ScalarParameter() {defaultValue=1, parameterName=UseEnd}) {constA=1}
  n22 = Saturate(None=Multiply(A=ScalarParameter() {defaultValue=2, parameterName=FalloffStartEnd}, B=Multiply(A=@n20, B=@n21)))
  n23 = Multiply(A=ScalarParameter() {defaultValue=1, group=WPOMotion, parameterName=SpiralSine1}, B=TextureCoordinate())
  n24 = Multiply(A=Time(), B=ScalarParameter() {defaultValue=8, parameterName=WPO Speed}) {constB=4}
  n25 = Add(
   A=Multiply(
    A=ComponentMask(None=@n23) {g=true}
    B=MaterialFunctionCall() {materialFunction=/Engine/Functions/Engine_MaterialFunctions02/WorldPositionOffset/ObjectScale}.Scale X
   )
   B=@n24
  )
  n26 = Constant3Vector() {constant=<rgba 0 1.5 1.5 0>}
  n27 = Multiply(A=ScalarParameter() {defaultValue=0.5, group=WPOMotion, parameterName=SpiralSine2}, B=TextureCoordinate())
  n28 = Add(
   A=Multiply(
    A=ComponentMask(None=@n27) {g=true}
    B=MaterialFunctionCall() {materialFunction=&ObjectScale}.Scale X
   )
   B=@n24
  )
  n29 = Constant3Vector() {constant=<rgba 0 1.5 -1.5 0>}
  n30 = Add(
   A=Multiply(A=Sine(None=@n25) {period=5}, B=Transform(None=@n26) {transformSourceType=TRANSFORMSOURCE_Local})
   B=Multiply(A=Cosine(None=@n28) {period=5}, B=Transform(None=@n29) {transformSourceType=TRANSFORMSOURCE_Local})
  )
  n31 = Multiply(A=ScalarParameter() {defaultValue=1, group=WPOMotion, parameterName=SpiralSine1}, B=TextureCoordinate())
  n32 = Add(
   A=Multiply(
    A=ComponentMask(None=@n31) {g=true}
    B=MaterialFunctionCall() {materialFunction=&ObjectScale}.Scale X
   )
   B=@n24
  )
  n33 = Constant3Vector() {constant=<rgba 0 -1.5 1.5 0>}
  n34 = Constant()
  n35 = Multiply(A=ScalarParameter() {defaultValue=0.5, group=WPOMotion, parameterName=SpiralSine2}, B=TextureCoordinate())
  n36 = Add(
   A=Multiply(
    A=ComponentMask(None=@n35) {g=true}
    B=MaterialFunctionCall() {materialFunction=&ObjectScale}.Scale X
   )
   B=@n24
  )
  n37 = AppendVector(A=Sine(None=@n36) {period=11}, B=Sine(None=@n36) {period=14.15})
  n38 = Constant3Vector() {constant=<rgba 0 1.5 -1.5 0>}
  n39 = Add(
   A=Multiply(A=Sine(None=@n32) {period=12.9}, B=Transform(None=@n33) {transformSourceType=TRANSFORMSOURCE_Local})
   B=Multiply(A=AppendVector(A=@n34, B=@n37), B=Transform(None=@n38) {transformSourceType=TRANSFORMSOURCE_Local})
  )
  n40 = Multiply(
   A=ComponentMask(None=TextureCoordinate()) {g=true}
   B=MaterialFunctionCall() {materialFunction=&ObjectScale}.Scale X
  )
  n41 = Multiply(A=Time(), B=ScalarParameter() {defaultValue=2, group=WPOMotion, parameterName=ReturnWiggleSpeed}) {constB=2}
  n42 = Multiply(
   A=Sine(None=Add(A=@n40, B=@n41)) {period=7}
   B=Transform(None=Constant3Vector() {constant=<rgba 0 0 3 0>}) {transformSourceType=TRANSFORMSOURCE_Local}
  )
  n43 = ScalarParameter() {group=WPOMotion, parameterName=ReturnWiggle}
  n44 = Multiply(
   A=Multiply(
    A=DynamicParameter() {defaultValue=<rgba 1 0 1 1>, paramNames [WiggleAmountMult, Erosion, Param3, Param4]}.WiggleAmountMult
    B=ScalarParameter() {defaultValue=10, group=WPOMotion, parameterName=WiggleAmount}
   )
   B=Add(A=StaticSwitchParameter(True=@n30, False=@n39) {defaultValue=true, group=WPOMotion, parameterName=SpiralWPO}, B=Multiply(A=@n42, B=@n43))
  )
  n45 = Divide(A=ObjectRadius(), B=ScalarParameter() {defaultValue=4096, group=WPOMotion, parameterName=MaxLength})
  n46 = SphereMask(
   A=ComponentMask(None=TextureCoordinate()) {g=true}
   B=Constant() {r=0.5}
   Radius=ScalarParameter() {defaultValue=0.5, group=WPOMotion, parameterName=SphereRadius}
   Hardness=Clamp(None=LinearInterpolate(Alpha=@n45))
  ) {attenuationRadius=0.5, hardnessPercent=50}
  n47 = VertexNormalWS()
  Add(
   A=Add(A=@n4, B=Multiply(A=@n18, B=@n22))
   B=StaticSwitchParameter(True=Multiply(A=@n44, B=@n46), False=Multiply(A=@n47) {constB=0}) {defaultValue=true, group=WPOMotion, parameterName=UseWPOMotion?}
  )
  ScalarParameter() {defaultValue=0.45, parameterName=Roughness}
  ScalarParameter() {defaultValue=0.1, parameterName=Specular}
  ScalarParameter() {defaultValue=0.91, parameterName=Metallic}
  ScalarParameter() {defaultValue=0.3, group=Clear Coat, parameterName=Clear Coat}
  ScalarParameter() {defaultValue=0.4, group=Clear Coat, parameterName=Clear Coat Roughness}
  Sine(None=@n28) {period=7}
  n48 = Multiply(
   A=TextureCoordinate()
   B=AppendVector(A=ScalarParameter() {defaultValue=2, parameterName=Tile_U}, B=ScalarParameter() {defaultValue=48, parameterName=Tile_V})
  )
  StaticSwitchParameter(
   True=Transform(None=@n3.Normal) {transformSourceType=TRANSFORMSOURCE_World, transformType=TRANSFORM_Tangent}
   False=Multiply(
    A=TextureSampleParameter2D(UVs=@n48) {parameterName=NormalMap, samplerType=SAMPLERTYPE_Normal, texture=/Game/Athena/Items/Consumables/ZipLine/T_Cable_N}.RGB
    B=VectorParameter() {parameterName=NormalStrength, defaultValue=<rgba 0.4 0.4 1 0>}.RGB
   )
  ) {parameterName=IsSplineThickening}
  n49 = TextureSampleParameter2D(UVs=@n48) {parameterName=BaseColor, texture=/Game/Athena/Items/Consumables/ZipLine/T_Cable}
  n50 = Multiply(
   A=Power(Base=@n49.RGB, Exp=ScalarParameter() {defaultValue=2, parameterName=Contrast})
   B=ScalarParameter() {defaultValue=1.3, parameterName=Brightness}
  )
  StaticSwitchParameter(
   True=LinearInterpolate(
    A=VectorParameter() {parameterName=BaseColorLow, defaultValue=<rgba 0.2199 0.2556 0.3142 0>}.RGB
    B=VectorParameter() {parameterName=BaseColorHigh, defaultValue=<rgba 0.4635 0.4635 0.4635 0>}.RGB
    Alpha=Saturate(None=@n50)
   )
   False=@n49.RGB
  ) {defaultValue=true, parameterName=UseBaseColorAdjustments}
  n51 = ScalarParameter() {defaultValue=1, group=Erosion, parameterName=Erosion Tile_U}
  n52 = ScalarParameter() {defaultValue=2, group=Erosion, parameterName=Erosion Tile_V}
  n53 = TextureSample(UVs=Multiply(A=TextureCoordinate(), B=AppendVector(A=@n51, B=@n52))) {samplerType=SAMPLERTYPE_LinearColor, texture=/Game/Characters/Player/Male/Medium/Bodies/M_MED_SRDrift/Textures/T_ElectricNoise}
  n54 = ComponentMask(None=TextureCoordinate()) {g=true}
  Subtract(
   A=LinearInterpolate(A=Multiply(A=@n53.R, B=@n54), B=@n54) {constAlpha=0.35}
   B=LinearInterpolate(Alpha=DynamicParameter() {defaultValue=<rgba 1 0 1 1>, paramNames [WiggleAmountMult, Erosion, Param3, Param4]}.Erosion) {constA=-0.3333, constB=0.6667}
  )
 ]
}
