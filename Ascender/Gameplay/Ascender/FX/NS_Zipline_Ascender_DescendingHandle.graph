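# Niagara effect for the zipline ascender's descending handle, producing friction-style smoke; only TopSmoke is enabled.
# TopSmoke uses two spawn groups for opposite sideways puffs that rise, grow and fade, with User.SpawnSpark as spawn probability.
# Disabled emitters hold bottom smoke, GPU top and bottom sparks with speed-stretched sprites, and a mesh skid spark.
# User.Speed and User.TurnSpeed are declared but drive none of the shown modules.
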NiagaraSystem /Ascender/Gameplay/Ascender/FX/NS_Zipline_Ascender_DescendingHandle
dependencies {
 usedDataInterfaces [&NiagaraDataInterfaceCurve, &NiagaraDataInterfaceVectorCurve]
 usedDynamicInputs []
 usedModules [
  &self:SystemSpawnScript.SystemScriptSource.SystemScriptGraph.NiagaraNodeAssignment_0.SetVariables_#1
  &SystemState
 ]
}
emitters {
 BottomSmoke {
  bEnabled=false
  emitterUpdateScript [
   &EmitterState()
   &SpawnRate(Spawn Probability=$User.SpawnSpark, SpawnRate=14, Use Spawn Probability=true)
   &SpawnRate(Spawn Probability=$User.SpawnSpark, SpawnGroup=1, SpawnRate=14, Use Spawn Probability=true) {module=SpawnRate001}
  ]
  particleSpawnScript [
   &InitializeParticle(
    Color Mode=::NewEnumerator0
    Lifetime=&RandomRangeFloat(Minimum=0.2, Maximum=0.55)
    Sprite Rotation Angle=&RandomRangeFloat(Minimum=0, Maximum=3)
    Sprite Rotation Mode=::NewEnumerator1
    Sprite Size Mode=::NewEnumerator1
    Uniform Sprite Size Max=25
    Uniform Sprite Size Min=18
    UsePositionOffset=false
    Write Material Random=true
   )
   &SphereLocation(
    MaskSpawnGroup=false
    Override Seed=false
    Sphere Radius=&RandomRangeFloat(Minimum=3, Maximum=6)
    Surface Only=false
    UseNonUniformScale=false
    UseOffset=false
    UseSphereOrientation=false
   ) {enabled=false}
   &SystemLocation(MaskSpawnGroup=true, Offset=<0 1 -14>, Spawn Group Mask=1)
   &SystemLocation(MaskSpawnGroup=false, Offset=<0 -1 -14>) {module=SystemLocation001}
   &AddVelocity(
    Cone Angle=10
    Cone Axis=&MakeCustomVectorFromBool(
     Bool=&SetBoolByIntComparison(A=1, B=$Emitter.SpawnGroup, Comparison Type=::NewEnumerator0)
     True Vector=<0 5 -2>
     False Vector=<-5 -5 -2>
    )
    Module.UseConeFalloff=false
    Module.UseVelDistribution=false
    Override Seed=false
    Velocity Mode=::NewEnumerator2
    Velocity Speed=&RandomRangeFloat(Minimum=50, Maximum=100)
   ) {module=AddVelocity003}
   &Color(Color=<rgba 0.2917 0.2579 0.2218 1.2>)
  ]
  particleUpdateScript [
   &Drag(Drag=4, Use Linear Drag=true, Use Rotational Drag=false)
   &GravityForce(Gravity=<0 0 -750>)
   &SolveForcesAndVelocity(Clamp Velocity=false, Limit Acceleration=false)
   &ParticleState()
   &DynamicMaterialParameters(
    Index 0 Param 1=&FloatFromCurve(FloatCurve=DataInterface() {propertyValues {Curve {keys [<at 0 0>,<at 1 0.6>]}}})
    Module.Param0WriteEnabled=true
    Module.Param1WriteEnabled=true
    Module.Param2WriteEnabled=true
    Module.Param3WriteEnabled=true
   )
   &ScaleColor(
    Scale Alpha=&FloatFromCurve(FloatCurve=DataInterface() {propertyValues {Curve {keys [<at 0 0>,<at 0.2037 1>,<at 1 0>]}}})
    ScaleA=true
    ScaleRGB=true
   )
   &ScaleSpriteSize(Scale Factor=&Vector2DFromFloat(Value=&FloatFromCurve(
      FloatCurve=DataInterface() {propertyValues {Curve {keys [{arriveTangent=1.9106, interpMode=RCIM_Cubic, leaveTangent=1.9106, tangentMode=RCTM_User, time=0, value=0}, <at 1 1 RCIM_Cubic>]}}}
      Scale Curve=1.5
     )))
  ]
  rendererClasses [&NiagaraSpriteRendererProperties]
  renderers [{rendererClass=&NiagaraSpriteRendererProperties}]
 }
 BottomSparks {
  bEnabled=false
  simTarget=GPUComputeSim
  emitterUpdateScript [
   &EmitterState()
   &SpawnRate(Spawn Probability=0.9, SpawnRate=50, Use Spawn Probability=true) {module=SpawnRate001}
   &SpawnRate(Spawn Probability=0.9, SpawnGroup=1, SpawnRate=50, Use Spawn Probability=true)
  ]
  particleSpawnScript [
   &InitializeParticle(
    Color=<rgba 5 1.6977 0.0741 1>
    Lifetime=&RandomRangeFloat(Minimum=0.05, Maximum=0.25, Evaluation Type=::NewEnumerator1)
    Sprite Size Max=<2.5 4>
    Sprite Size Min=<0.5 0.5>
    Sprite Size Mode=::NewEnumerator2
    UsePositionOffset=false
    Write Material Random=true
   ) {module=InitializeParticle001}
   &ShapeLocation(
    MaskSpawnGroup=false
    Offset Mode=::NewEnumerator0
    Override Seed=false
    Shape Origin=$Particles.Position
    Sphere Radius=5
    UseOwnerScale=true
   ) {enabled=false}
   &SystemLocation(MaskSpawnGroup=true, Offset=<0 2.5 -14>)
   &SystemLocation(MaskSpawnGroup=true, Offset=<0 -2.5 -14>, Spawn Group Mask=1) {module=SystemLocation001}
   &AddVelocity(
    Cone Angle=30
    Cone Axis=&MakeCustomVectorFromBool(
     Bool=&SetBoolByIntComparison(A=1, B=$Emitter.SpawnGroup, Comparison Type=::NewEnumerator0)
     True Vector=<0 5 -3>
     False Vector=<1 -5 -3>
    )
    Module.UseConeFalloff=false
    Module.UseVelDistribution=false
    Override Seed=false
    Velocity Mode=::NewEnumerator2
    Velocity Speed=&RandomRangeFloat(Minimum=100, Maximum=250)
   )
  ]
  particleUpdateScript [
   &ParticleState()
   &AccelerationForce(Acceleration=<0 0 -150>)
   &GravityForce(Gravity=<0 0 -1000>)
   &SolveForcesAndVelocity(Clamp Velocity=false, Limit Acceleration=false)
   &ScaleColor(
    Scale Alpha=&FloatFromCurve(FloatCurve=DataInterface() {propertyValues {Curve {keys [{arriveTangent=-6.8352, interpMode=RCIM_Cubic, time=0, value=2}, <at 0.2 1>]}}})
    Scale RGB=&VectorFromCurve(VectorCurve=DataInterface() {propertyValues {
       XCurve {keys [<at 0 16 RCIM_Cubic>,<at 0.8 1 RCIM_Cubic>]}
       YCurve {keys [<at 0 16 RCIM_Cubic>,<at 0.8 0.1249 RCIM_Cubic>]}
       ZCurve {keys [<at 0 16 RCIM_Cubic>,<at 0.8 0.2547 RCIM_Cubic>]}
      }})
    ScaleA=true
    ScaleRGB=true
   )
   &ScaleSpriteSizeBySpeed(Max Scale Factor=<1.5 20>, Min Scale Factor=<0.5 0.5>, Velocity Threshold=1250) {module=ScaleSpriteSizeBySpeed001}
  ]
  rendererClasses [&NiagaraSpriteRendererProperties]
  renderers [{rendererClass=&NiagaraSpriteRendererProperties}]
 }
 SkidSharp {
  bEnabled=false
  emitterUpdateScript [
   &EmitterState(Life Cycle Mode=::NewEnumerator1, Loop Duration=1, UseLoopDelay=true)
   &SpawnRate(Spawn Probability=$User.SpawnSpark, SpawnRate=15, Use Spawn Probability=true)
   &SpawnRate(Spawn Probability=$User.SpawnSpark, SpawnGroup=1, SpawnRate=15, Use Spawn Probability=false) {module=SpawnRate001}
  ]
  particleSpawnScript [
   &InitializeParticle(
    Color=<rgba 3 0.9113 0 1>
    Lifetime=&RandomRangeFloat(Minimum=0.12, Maximum=0.16, Evaluation Type=::NewEnumerator1)
    Mesh Scale Mode=::NewEnumerator3
    Mesh Uniform Scale=0.65
    UsePositionOffset=false
    Write Material Random=true
   ) {module=InitializeParticle001}
   &AddVelocity(Velocity=&RandomRangeVector(Minimum=<0 -25 0>, Maximum=<0 -5 0>, Evaluation Type=::NewEnumerator1))
   &DynamicMaterialParameters(
    Index 0 Param 3=&RandomRangeFloat(Minimum=1, Maximum=1.5, Evaluation Type=::NewEnumerator1)
    Module.Param0WriteEnabled=false
    Module.Param1WriteEnabled=false
    Module.Param2WriteEnabled=true
    Module.Param3WriteEnabled=false
   ) {module=DynamicMaterialParameters001}
   &SystemLocation(MaskSpawnGroup=false, Offset=<0 2 -10>)
   &SystemLocation(MaskSpawnGroup=true, Offset=<0 -2 -10>, Spawn Group Mask=1) {module=SystemLocation001}
   &InitialMeshOrientation(
    Rotation=&RandomRangeVector(Minimum=<0 0 0.1>, Maximum=<0 -0.1 -0.1>)
    Use Orientation Vector=true
    Use Rotation Vector=true
   )
   &AddVelocityFromPoint(Module.Use Falloff Distance=false, Velocity Strength=&RandomRangeFloat(Minimum=120, Maximum=250))
  ]
  particleUpdateScript [
   /SetVariables(Particles.SpriteSize=<1 1>) {enabled=false, isSetParameters=true}
   &ParticleState()
   &Drag(Drag=4, Use Linear Drag=true, Use Rotational Drag=false)
   &SolveForcesAndVelocity(Clamp Velocity=false, Limit Acceleration=false)
   &ScaleColor(
    Scale Alpha=&FloatFromCurve(FloatCurve=DataInterface() {propertyValues {Curve {keys [<at 0.4 1 RCIM_Cubic>, {interpMode=RCIM_Cubic, leaveTangent=-3.0416, time=1, value=0}]}}})
    ScaleA=true
    ScaleRGB=true
   )
   &DynamicMaterialParameters(
    Index 0 Param 1=&FloatFromCurve(FloatCurve=DataInterface() {propertyValues {Curve {keys [<at 0.1 0.1 RCIM_Cubic>,<at 0.8 1>]}}})
    Index 0 Param 2=2
    Index 0 Param 4=&FloatFromCurve(FloatCurve=DataInterface() {propertyValues {Curve {keys [<at 0 1>,<at 1 3>]}}})
    Module.Param0WriteEnabled=true
    Module.Param1WriteEnabled=true
    Module.Param2WriteEnabled=false
    Module.Param3WriteEnabled=true
   )
   &ScaleMeshSize(Scale Factor=&VectorFromFloat(Value=&FloatFromCurve(
      FloatCurve=DataInterface() {propertyValues {Curve {keys [{arriveTangent=2.1642, interpMode=RCIM_Cubic, leaveTangent=2.1642, tangentMode=RCTM_User, time=0, value=0}, <at 1 1 RCIM_Cubic>]}}}
      Scale Curve=0.65
     )))
  ]
  rendererClasses [&NiagaraMeshRendererProperties]
  renderers [{rendererClass=&NiagaraMeshRendererProperties, rendererIndex=0}]
 }
 TopSmoke {
  emitterUpdateScript [
   &EmitterState()
   &SpawnRate(Spawn Probability=$User.SpawnSpark, SpawnRate=10, Use Spawn Probability=true)
   &SpawnRate(Spawn Probability=$User.SpawnSpark, SpawnGroup=1, SpawnRate=10, Use Spawn Probability=true) {module=SpawnRate001}
  ]
  particleSpawnScript [
   &InitializeParticle(
    Color Mode=::NewEnumerator0
    Lifetime=&RandomRangeFloat(Minimum=0.2, Maximum=0.48)
    Sprite Rotation Angle=&RandomRangeFloat(Minimum=0, Maximum=3)
    Sprite Rotation Mode=::NewEnumerator1
    Sprite Size Mode=::NewEnumerator1
    Uniform Sprite Size Max=25
    Uniform Sprite Size Min=18
    UsePositionOffset=false
    Write Material Random=true
   )
   &SphereLocation(
    MaskSpawnGroup=false
    Override Seed=false
    Sphere Radius=&RandomRangeFloat(Minimum=3, Maximum=6)
    Surface Only=false
    UseNonUniformScale=false
    UseOffset=false
    UseSphereOrientation=false
   ) {enabled=false}
   &SystemLocation(MaskSpawnGroup=false, Offset=<0 2 0>)
   &SystemLocation(MaskSpawnGroup=true, Offset=<0 -2 0>, Spawn Group Mask=1) {module=SystemLocation001}
   &AddVelocity(
    Cone Angle=10
    Cone Axis=&MakeCustomVectorFromBool(
     Bool=&SetBoolByIntComparison(A=1, B=$Emitter.SpawnGroup, Comparison Type=::NewEnumerator0)
     True Vector=<0 5 -2>
     False Vector=<1 -5 -2>
    )
    Module.UseConeFalloff=false
    Module.UseVelDistribution=false
    Override Seed=false
    Velocity Mode=::NewEnumerator2
    Velocity Speed=&RandomRangeFloat(Minimum=50, Maximum=100)
   ) {module=AddVelocity003}
   &Color(Color=<rgba 0.2917 0.2579 0.2218 1.1>)
  ]
  particleUpdateScript [
   &Drag(Drag=4, Use Linear Drag=true, Use Rotational Drag=false)
   &GravityForce(Gravity=<0 0 750>)
   &SolveForcesAndVelocity(Clamp Velocity=false, Limit Acceleration=false)
   &ParticleState()
   &DynamicMaterialParameters(
    Index 0 Param 1=&FloatFromCurve(FloatCurve=DataInterface() {propertyValues {Curve {keys [<at 0 0>,<at 1 0.6>]}}})
    Module.Param0WriteEnabled=true
    Module.Param1WriteEnabled=true
    Module.Param2WriteEnabled=true
    Module.Param3WriteEnabled=true
   )
   &ScaleColor(
    Scale Alpha=&FloatFromCurve(FloatCurve=DataInterface() {propertyValues {Curve {keys [<at 0 0>,<at 0.2037 1>,<at 1 0>]}}})
    ScaleA=true
    ScaleRGB=true
   )
   &ScaleSpriteSize(Scale Factor=&Vector2DFromFloat(Value=&FloatFromCurve(
      FloatCurve=DataInterface() {propertyValues {Curve {keys [{arriveTangent=1.9106, interpMode=RCIM_Cubic, leaveTangent=1.9106, tangentMode=RCTM_User, time=0, value=0}, <at 1 1 RCIM_Cubic>]}}}
      Scale Curve=1.5
     )))
  ]
  rendererClasses [&NiagaraSpriteRendererProperties]
  renderers [{rendererClass=&NiagaraSpriteRendererProperties}]
 }
 TopSparks {
  bEnabled=false
  simTarget=GPUComputeSim
  emitterUpdateScript [
   &EmitterState()
   &SpawnRate(Spawn Probability=0.9, SpawnRate=40, Use Spawn Probability=true) {module=SpawnRate001}
   &SpawnRate(Spawn Probability=0.9, SpawnGroup=1, SpawnRate=40, Use Spawn Probability=true)
  ]
  particleSpawnScript [
   &InitializeParticle(
    Color=<rgba 5 1.6977 0.0741 1>
    Lifetime=&RandomRangeFloat(Minimum=0.05, Maximum=0.2, Evaluation Type=::NewEnumerator1)
    Sprite Size Max=<2.5 4>
    Sprite Size Min=<0.25 0.5>
    Sprite Size Mode=::NewEnumerator2
    UsePositionOffset=false
    Write Material Random=true
   ) {module=InitializeParticle001}
   &ShapeLocation(
    MaskSpawnGroup=false
    Offset Mode=::NewEnumerator0
    Override Seed=false
    Shape Origin=$Particles.Position
    Sphere Radius=5
    UseOwnerScale=true
   ) {enabled=false}
   &SystemLocation(MaskSpawnGroup=true, Offset=<0 2.5 14.25>)
   &SystemLocation(MaskSpawnGroup=true, Offset=<0 -2.5 14.25>, Spawn Group Mask=1) {module=SystemLocation001}
   &AddVelocity(
    Cone Angle=30
    Cone Axis=&MakeCustomVectorFromBool(
     Bool=&SetBoolByIntComparison(A=1, B=$Emitter.SpawnGroup, Comparison Type=::NewEnumerator0)
     True Vector=<0 5 -1>
     False Vector=<1 -5 -1>
    )
    Module.UseConeFalloff=false
    Module.UseVelDistribution=false
    Override Seed=false
    Velocity Mode=::NewEnumerator2
    Velocity Speed=&RandomRangeFloat(Minimum=100, Maximum=250)
   )
  ]
  particleUpdateScript [
   &ParticleState()
   &AccelerationForce(Acceleration=<0 0 -150>)
   &GravityForce(Gravity=<0 0 -1200>)
   &SolveForcesAndVelocity(Clamp Velocity=false, Limit Acceleration=false)
   &ScaleColor(
    Scale Alpha=&FloatFromCurve(FloatCurve=DataInterface() {propertyValues {Curve {keys [{arriveTangent=-6.8352, interpMode=RCIM_Cubic, time=0, value=2}, <at 0.2 1>]}}})
    Scale RGB=&VectorFromCurve(VectorCurve=DataInterface() {propertyValues {
       XCurve {keys [<at 0 16 RCIM_Cubic>,<at 0.8 1 RCIM_Cubic>]}
       YCurve {keys [<at 0 16 RCIM_Cubic>,<at 0.8 0.1249 RCIM_Cubic>]}
       ZCurve {keys [<at 0 16 RCIM_Cubic>,<at 0.8 0.2547 RCIM_Cubic>]}
      }})
    ScaleA=true
    ScaleRGB=true
   )
   &ScaleSpriteSizeBySpeed(Max Scale Factor=<1.5 25>, Min Scale Factor=<0.5 0.5>, Velocity Threshold=1250) {module=ScaleSpriteSizeBySpeed001}
  ]
  rendererClasses [&NiagaraSpriteRendererProperties]
  renderers [{rendererClass=&NiagaraSpriteRendererProperties}]
 }
}
summary {
 systemName=NS_Zipline_Ascender_DescendingHandle
 emitters [
  {bEnabled=false, emitterName=SkidSharp, rendererClasses [&NiagaraMeshRendererProperties]}
  {bEnabled=false, emitterName=BottomSmoke, rendererClasses [&NiagaraSpriteRendererProperties]}
  {emitterName=TopSmoke, rendererClasses [&NiagaraSpriteRendererProperties]}
  {bEnabled=false, emitterName=TopSparks, simTarget=GPUComputeSim, rendererClasses [&NiagaraSpriteRendererProperties]}
  {bEnabled=false, emitterName=BottomSparks, simTarget=GPUComputeSim, rendererClasses [&NiagaraSpriteRendererProperties]}
 ]
 userVariables [
  User.SpawnSpark {type=NiagaraFloat, defaultValue=1, description=""}
  User.Speed {type=NiagaraFloat, defaultValue=0, description=""}
  User.TurnSpeed {type=NiagaraFloat, defaultValue=0, description=""}
 ]
}
