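; Gameplay ability for Ascender ziplines (Smash variant), with helpers for weapon holstering and montage choice.
; DoHolster and DoUnholster track bHolsterApplied, and a cvar-gated fail-safe unholsters if that flag is still set.
; GetAscenderMontage prefers a montage from the pawn's AscenderMontageProvider component, otherwise picking a default by character gender.
; IsPlayerUsingHandle compares the player to the zipline's handle user, selecting ascending or descending animation.
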(fn DoFailSafeUnholster
  (bind _playerpawn (Variables|Default|GetPlayerPawn))
  (bind _returnvalue (Development|GetConsoleVariableBoolValue "Ascender.UseHolsterFailSafe"))
  (if (and _returnvalue (Variables|Default|GetbHolsterApplied))
    (Variables|Default|SetbHolsterApplied false)
    (bind _returnvalue_2 (Development|GetConsoleVariableBoolValue "Ascender.UseSafeHolster"))
    (if _returnvalue_2
      (Weapon|SetHolsterWeapon _playerpawn 0 false false "Ascender")
      (else
        (Weapon|SetHolsterWeapon _playerpawn 0 false false "Ascender")))))

(fn DoHolster
  (bind _playerpawn (Variables|Default|GetPlayerPawn))
  (Variables|Default|SetbHolsterApplied true)
  (bind _returnvalue (Development|GetConsoleVariableBoolValue "Ascender.UseSafeHolster"))
  (if _returnvalue
    (Weapon|SetHolsterWeapon _playerpawn 0 true false "Ascender")
    (else
      (Weapon|SetHolsterWeapon _playerpawn 0 true false "Ascender"))))

(fn DoUnholster
  (bind _playerpawn (Variables|Default|GetPlayerPawn))
  (Variables|Default|SetbHolsterApplied false)
  (bind _returnvalue (Development|GetConsoleVariableBoolValue "Ascender.UseSafeHolster"))
  (if _returnvalue
    (Weapon|SetHolsterWeapon _playerpawn 0 false true "Ascender")
    (else
      (Weapon|SetHolsterWeapon _playerpawn 0 false true "Ascender"))))

(fn GetAscenderMontage
  (bind _playerpawn (Variables|Default|GetPlayerPawn))
  (bind _returnvalue (Actor|GetComponentbyClass _playerpawn "/Script/AscenderCodeRuntime.FortPawnComponent_AscenderMontageProvider"))
  (Utilities|FlowControl|Sequence
    (:then_0
      ;[1 1 1] Check for Override Component. This allows specific Pawns to use their own animations
      (if (Utilities|IsValid _returnvalue :!class /Script/Engine.KismetSystemLibrary)
        (return (Class|FortPawnComponentAscenderMontageProvider|GetSuitableMontage _returnvalue (CallFunction|IsPlayerUsingHandle)))))
    (:then_1
      ;[1 1 1] Return suitable default montage if no override
      (return (Utilities|Select :Index (Customization|GetCharacterGender (Variables|Default|GetPlayerPawn)) :Invalid (select (CallFunction|IsPlayerUsingHandle) (Variables|Default|GetM_AscendingAnim) (Variables|Default|GetM_DescendingAnim)) :Male (select (CallFunction|IsPlayerUsingHandle) (Variables|Default|GetM_AscendingAnim) (Variables|Default|GetM_DescendingAnim)) :Female (select (CallFunction|IsPlayerUsingHandle) (Variables|Default|GetF_AscendingAnim) (Variables|Default|GetF_DescendingAnim)) :Both (select (CallFunction|IsPlayerUsingHandle) (Variables|Default|GetM_AscendingAnim) (Variables|Default|GetM_DescendingAnim)))))))

(fn IsPlayerUsingHandle
  (bind _playerpawn (Variables|Default|GetPlayerPawn))
  (bind _ascenderzipline (Variables|Default|GetAscenderZipline))
  (Utilities|IsValid _ascenderzipline
    (:Is%20Valid
      (Utilities|IsValid _playerpawn
        (:Is%20Valid
          (return (== _playerpawn (Class|FortAscenderZipline|GetPawnUsingHandle _ascenderzipline))))
        (:Is%20Not%20Valid
          _)))
    (:Is%20Not%20Valid
      _))
  (return false))
