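# Sound cue that plays when an Ascender gadget is destroyed.
# A Random node picks one of five equally weighted destruction wave variants and is excluded from branch culling.
# Its output feeds a Modulator at the root for per-play pitch and volume variation.
# The cue uses a dedicated destroyed attenuation asset, the gadgets and items sound class, and a slightly boosted volume.
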SoundCue /Ascender/Gameplay/Ascender/Sounds/Ascender_Destroyed_Cue
root=Modulator_0
nodes [
 {id=Modulator_0, children [Random_0]}
 {
  id=Random_0
  children [WavePlayer_0,WavePlayer_1,WavePlayer_2,WavePlayer_3,WavePlayer_4]
  properties {bShouldExcludeFromBranchCulling=true, weights [1,1,1,1,1]}
 }
 {id=WavePlayer_0, properties {soundWaveAssetPtr=/Ascender/Gameplay/Ascender/Sounds/wav/Ascender_Destroyed_05}}
 {id=WavePlayer_1, properties {soundWaveAssetPtr=/Ascender/Gameplay/Ascender/Sounds/wav/Ascender_Destroyed_01}}
 {id=WavePlayer_2, properties {soundWaveAssetPtr=/Ascender/Gameplay/Ascender/Sounds/wav/Ascender_Destroyed_02}}
 {id=WavePlayer_3, properties {soundWaveAssetPtr=/Ascender/Gameplay/Ascender/Sounds/wav/Ascender_Destroyed_03}}
 {id=WavePlayer_4, properties {soundWaveAssetPtr=/Ascender/Gameplay/Ascender/Sounds/wav/Ascender_Destroyed_04}}
]
properties {
 attenuationSettings=/Game/Sounds/Attenuation/Fort_Ascender_Destroyed_Att
 maxDistance=6500
 soundClassObject=&Fort_Gadgets_Items
 volumeMultiplier=1.2
}
