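# Sound cue for the Ascender handle's end-of-descent skid, played as a spatialized gadget sound with limited concurrency.
# Branches on IsLocal, then IsEnemy, to set loudness: enemy or unset is loudest, local is moderate, and non-enemy remote players are quietest.
# Each path randomly picks one of the same two handle waves through a Modulator for slight variation.
# The random nodes are excluded from branch culling.
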SoundCue /Ascender/Gameplay/Ascender/Sounds/Ascender_Handle_EndDecent_Skid_Cue
root=Branch_0
nodes [
 {id=Branch_0, children [Mixer_5,Branch_1,Branch_1], properties {boolParameterName=IsLocal}}
 {id=Branch_1, children [Mixer_1,Mixer_2,Mixer_1], properties {boolParameterName=IsEnemy}}
 {id=Mixer_1, children [Modulator_1], properties {inputVolume [1]}}
 {id=Mixer_2, children [Modulator_2], properties {inputVolume [0.25]}}
 {id=Mixer_5, children [Modulator_0], properties {inputVolume [0.65]}}
 {id=Modulator_0, children [Random_0], properties {pitchMax=1}}
 {id=Modulator_1, children [Random_1], properties {pitchMax=1}}
 {id=Modulator_2, children [Random_2], properties {pitchMax=1}}
 {id=Random_0, children [WavePlayer_1,WavePlayer_2], properties {bShouldExcludeFromBranchCulling=true, weights [1,1]}}
 {id=Random_1, children [WavePlayer_0,WavePlayer_3], properties {bShouldExcludeFromBranchCulling=true, weights [1,1]}}
 {id=Random_2, children [WavePlayer_4,WavePlayer_5], properties {bShouldExcludeFromBranchCulling=true, weights [1,1]}}
 {id=WavePlayer_0, properties {soundWaveAssetPtr=/Ascender/Gameplay/Ascender/Sounds/wav/Ascender_Ascender_End_Descent_Handle_02}}
 {id=WavePlayer_1, properties {soundWaveAssetPtr=&Ascender_Ascender_End_Descent_Handle_02}}
 {id=WavePlayer_2, properties {soundWaveAssetPtr=/Ascender/Gameplay/Ascender/Sounds/wav/Ascender_Ascender_End_Descent_Handle_01}}
 {id=WavePlayer_3, properties {soundWaveAssetPtr=&Ascender_Ascender_End_Descent_Handle_01}}
 {id=WavePlayer_4, properties {soundWaveAssetPtr=&Ascender_Ascender_End_Descent_Handle_02}}
 {id=WavePlayer_5, properties {soundWaveAssetPtr=&Ascender_Ascender_End_Descent_Handle_01}}
]
properties {
 attenuationSettings=/Game/Sounds/Attenuation/Fort_Ascender_Handle_Reset_Att
 bOverrideConcurrency=true
 maxDistance=6000
 soundClassObject=&Fort_Gadgets
 volumeMultiplier=0.25
 attenuationOverrides {falloffDistance=1536, stereoSpread=0, attenuationShapeExtents {x=512}}
 concurrencyOverrides {maxCount=2}
}
