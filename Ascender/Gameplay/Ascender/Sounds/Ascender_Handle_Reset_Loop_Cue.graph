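# Looping sound cue for the Ascender handle reset, built from the Ascender descend loop wave pitched down by a fixed amount.
# An IsLocal branch plays the local player's loop at full volume; otherwise a nested IsEnemy branch keeps enemy or unset cases at full volume and drops non-enemies to quarter volume.
# A cue-level volume trim, dedicated attenuation asset and gadgets/items sound class apply to all paths.
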SoundCue /Ascender/Gameplay/Ascender/Sounds/Ascender_Handle_Reset_Loop_Cue
root=Branch_0
nodes [
 {id=Branch_0, children [Mixer_2,Branch_1,Branch_1], properties {boolParameterName=IsLocal}}
 {id=Branch_1, children [Mixer_1,Mixer_0,Mixer_1], properties {boolParameterName=IsEnemy}}
 {id=Mixer_0, children [Modulator_1], properties {inputVolume [0.25]}}
 {id=Mixer_1, children [Modulator_0], properties {inputVolume [1]}}
 {id=Mixer_2, children [Modulator_2], properties {inputVolume [1]}}
 {id=Modulator_0, children [WavePlayer_0], properties {pitchMax=0.65, pitchMin=0.65}}
 {id=Modulator_1, children [WavePlayer_1], properties {pitchMax=0.65, pitchMin=0.65}}
 {id=Modulator_2, children [WavePlayer_2], properties {pitchMax=0.65, pitchMin=0.65}}
 {id=WavePlayer_0, properties {bLooping=true, soundWaveAssetPtr=/Ascender/Gameplay/Ascender/Sounds/wav/Ascender_Descend_Loop}}
 {id=WavePlayer_1, properties {bLooping=true, soundWaveAssetPtr=&Ascender_Descend_Loop}}
 {id=WavePlayer_2, properties {bLooping=true, soundWaveAssetPtr=&Ascender_Descend_Loop}}
]
properties {
 attenuationSettings=/Game/Sounds/Attenuation/Fort_Ascender_Handle_Reset_Att
 maxDistance=6000
 soundClassObject=&Fort_Gadgets_Items
 volumeMultiplier=0.504
}
