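# Sound cue named for the looping sound of a player ascending on an Ascender, though it holds no nodes and has no root, so no wave is connected.
# Only playback settings are present: the shared Ascender attenuation asset, the gadget items sound class, and a concurrency override that caps simultaneous instances.
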SoundCue /Ascender/Gameplay/Ascender/Sounds/Ascender_Player_Accend_Loop_Cue
root=null
nodes []
properties {
 attenuationSettings=/Game/Sounds/Attenuation/Fort_Ascender_Att
 bOverrideConcurrency=true
 maxDistance=10200
 soundClassObject=&Fort_Gadgets_Items
 volumeMultiplier=1
 concurrencyOverrides {maxCount=3}
}
