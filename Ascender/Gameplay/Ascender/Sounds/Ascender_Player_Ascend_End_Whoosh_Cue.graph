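# Sound cue for the hop-off whoosh played when a player finishes riding an Ascender.
# An IsLocal branch plays the local path, while remote players pass to an IsEnemy branch that plays for enemies or unset values and is silent when false.
# Both audible paths pick randomly among the same three whoosh waves through separate modulators.
# Attenuation and concurrency are overridden on the cue to limit range and simultaneous voices.
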SoundCue /Ascender/Gameplay/Ascender/Sounds/Ascender_Player_Ascend_End_Whoosh_Cue
root=Branch_0
nodes [
 {id=Branch_0, children [Mixer_5,Branch_1,Branch_1], properties {boolParameterName=IsLocal}}
 {id=Branch_1, children [Mixer_1,null,Mixer_1], properties {boolParameterName=IsEnemy}}
 {id=Mixer_1, children [Modulator_0], properties {inputVolume [1]}}
 {id=Mixer_5, children [Modulator_1], properties {inputVolume [1]}}
 {id=Modulator_0, children [Random_1]}
 {id=Modulator_1, children [Random_0]}
 {id=Random_0, children [WavePlayer_1,WavePlayer_2,WavePlayer_3], properties {weights [1,1,1]}}
 {id=Random_1, children [WavePlayer_9, WavePlayer_10, WavePlayer_11], properties {weights [1,1,1]}}
 {id=WavePlayer_1, properties {soundWaveAssetPtr=/Ascender/Gameplay/Ascender/Sounds/wav/Ascender_Ascender_Player_HopOff_Whoosh_02}}
 {id=WavePlayer_10, properties {soundWaveAssetPtr=/Ascender/Gameplay/Ascender/Sounds/wav/Ascender_Ascender_Player_HopOff_Whoosh_03}}
 {id=WavePlayer_11, properties {soundWaveAssetPtr=/Ascender/Gameplay/Ascender/Sounds/wav/Ascender_Ascender_Player_HopOff_Whoosh}}
 {id=WavePlayer_2, properties {soundWaveAssetPtr=&Ascender_Ascender_Player_HopOff_Whoosh_03}}
 {id=WavePlayer_3, properties {soundWaveAssetPtr=&Ascender_Ascender_Player_HopOff_Whoosh}}
 {id=WavePlayer_9, properties {soundWaveAssetPtr=&Ascender_Ascender_Player_HopOff_Whoosh_02}}
]
properties {
 attenuationSettings=/Game/Sounds/Attenuation/Fort_Zipline_Attachment_Att
 bOverrideAttenuation=true
 bOverrideConcurrency=true
 maxDistance=1900
 soundClassObject=&Fort_Gadgets_Items
 volumeMultiplier=0.45
 attenuationOverrides {falloffDistance=1500}
 concurrencyOverrides {maxCount=3}
}
