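# Sound cue for the start of a player's ascend on an Ascender, playing a single attach sound.
# IsLocal and IsEnemy branches route that same wave through mixers: full volume for local player, enemy or unknown affiliation, much quieter for non-local non-enemies.
# It uses zipline attachment attenuation and the gadgets sound class, with per-owner concurrency limited to one voice that stops the oldest.
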SoundCue /Ascender/Gameplay/Ascender/Sounds/Ascender_Player_Ascend_Start_Cue
root=Branch_0
nodes [
 {id=Branch_0, children [Mixer_5,Branch_1,Branch_1], properties {boolParameterName=IsLocal}}
 {id=Branch_1, children [Mixer_1,Mixer_4,Mixer_1], properties {boolParameterName=IsEnemy}}
 {id=Mixer_1, children [WavePlayer_0], properties {inputVolume [1]}}
 {id=Mixer_4, children [WavePlayer_2], properties {inputVolume [0.25]}}
 {id=Mixer_5, children [WavePlayer_1], properties {inputVolume [1]}}
 {id=WavePlayer_0, properties {soundWaveAssetPtr=/Ascender/Gameplay/Ascender/Sounds/wav/Ascender_Attach_01}}
 {id=WavePlayer_1, properties {soundWaveAssetPtr=&Ascender_Attach_01}}
 {id=WavePlayer_2, properties {soundWaveAssetPtr=&Ascender_Attach_01}}
]
properties {
 attenuationSettings=/Game/Sounds/Attenuation/Fort_Zipline_Attachment_Att
 bOverrideConcurrency=true
 maxDistance=10000
 soundClassObject=&Fort_Gadgets_Items
 volumeMultiplier=0.6
 concurrencyOverrides {bLimitToOwner=true, maxCount=1, resolutionRule=StopOldest, retriggerTime=0.3, voiceStealReleaseTime=0.2}
}
