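# Sound cue for the player stopping an Ascender ascent, playing a single detach wave.
# Branches on IsLocal, then IsEnemy, send it through separate mixer and modulator chains, keeping full input volume for local, enemy or unset-enemy cases and lowering it for non-local friendlies.
# It uses zipline attachment attenuation, the gadgets and items sound class, and a concurrency override limiting simultaneous instances.
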SoundCue /Ascender/Gameplay/Ascender/Sounds/Ascender_Player_Ascend_Stop_Cue
root=Branch_0
nodes [
 {id=Branch_0, children [Mixer_5,Branch_1,Branch_1], properties {boolParameterName=IsLocal}}
 {id=Branch_1, children [Mixer_1,Mixer_4,Mixer_1], properties {boolParameterName=IsEnemy}}
 {id=Mixer_1, children [Modulator_1], properties {inputVolume [1]}}
 {id=Mixer_4, children [Modulator_2], properties {inputVolume [0.25]}}
 {id=Mixer_5, children [Modulator_0], properties {inputVolume [1]}}
 {id=Modulator_0, children [WavePlayer_1]}
 {id=Modulator_1, children [WavePlayer_0]}
 {id=Modulator_2, children [WavePlayer_2]}
 {id=WavePlayer_0, properties {soundWaveAssetPtr=/Ascender/Gameplay/Ascender/Sounds/wav/Ascender_Detach_01}}
 {id=WavePlayer_1, properties {soundWaveAssetPtr=&Ascender_Detach_01}}
 {id=WavePlayer_2, properties {soundWaveAssetPtr=&Ascender_Detach_01}}
]
properties {
 attenuationSettings=/Game/Sounds/Attenuation/Fort_Zipline_Attachment_Att
 bOverrideConcurrency=true
 maxDistance=10000
 soundClassObject=&Fort_Gadgets_Items
 volumeMultiplier=0.65
 concurrencyOverrides {maxCount=3}
}
