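# Looping player descent sound cue for the Ascender, spatialized with Ascender attenuation.
# Branches on IsLocal, then IsEnemy, play non-local friendly sounds at quarter mix volume, while local, enemy and unset cases stay at full level.
# Each path layers descend and ascend loops through continuous modulators driven by SpeedFactor, with opposite pitch slopes.
# Both layers' volume ranges fall as input rises, with the descend layer using the absolute SpeedFactor for volume.
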SoundCue /Ascender/Gameplay/Ascender/Sounds/Ascender_Player_Decent_Loop_Cue
root=Branch_0
nodes [
 {id=Branch_0, children [Mixer_5,Branch_1,Branch_1], properties {boolParameterName=IsLocal}}
 {id=Branch_1, children [Mixer_1,Mixer_0,Mixer_1], properties {boolParameterName=IsEnemy}}
 {id=Mixer_0, children [Mixer_6], properties {inputVolume [0.25]}}
 {id=Mixer_1, children [Mixer_3], properties {inputVolume [1]}}
 {id=Mixer_2, children [ModulatorContinuous_2, ModulatorContinuous_3], properties {inputVolume [1.2,1.8]}}
 {id=Mixer_3, children [ModulatorContinuous_4, ModulatorContinuous_5], properties {inputVolume [1.2,1.8]}}
 {id=Mixer_5, children [Mixer_2], properties {inputVolume [1]}}
 {id=Mixer_6, children [ModulatorContinuous_6, ModulatorContinuous_7], properties {inputVolume [1.2,1.8]}}
 {
  id=ModulatorContinuous_2
  children [WavePlayer_1]
  properties {
   pitchModulationParams {default=0, maxOutput=0.75, minInput=-1, minOutput=2, parameterName=SpeedFactor}
   volumeModulationParams {default=0, maxOutput=0, minInput=-1, minOutput=1, paramMode=MPM_Abs, parameterName=SpeedFactor}
  }
 }
 {
  id=ModulatorContinuous_3
  children [WavePlayer_2]
  properties {
   pitchModulationParams {default=0, maxOutput=2, minInput=-1, minOutput=0.8, parameterName=SpeedFactor}
   volumeModulationParams {default=0, maxOutput=0, minInput=-1, minOutput=1, parameterName=SpeedFactor}
  }
 }
 {
  id=ModulatorContinuous_4
  children [WavePlayer_3]
  properties {
   pitchModulationParams {default=0, maxOutput=0.75, minInput=-1, minOutput=2, parameterName=SpeedFactor}
   volumeModulationParams {default=0, maxOutput=0, minInput=-1, minOutput=1, paramMode=MPM_Abs, parameterName=SpeedFactor}
  }
 }
 {
  id=ModulatorContinuous_5
  children [WavePlayer_4]
  properties {
   pitchModulationParams {default=0, maxOutput=2, minInput=-1, minOutput=0.8, parameterName=SpeedFactor}
   volumeModulationParams {default=0, maxOutput=0, minInput=-1, minOutput=1, parameterName=SpeedFactor}
  }
 }
 {
  id=ModulatorContinuous_6
  children [WavePlayer_6]
  properties {
   pitchModulationParams {default=0, maxOutput=0.75, minInput=-1, minOutput=2, parameterName=SpeedFactor}
   volumeModulationParams {default=0, maxOutput=0, minInput=-1, minOutput=1, paramMode=MPM_Abs, parameterName=SpeedFactor}
  }
 }
 {
  id=ModulatorContinuous_7
  children [WavePlayer_7]
  properties {
   pitchModulationParams {default=0, maxOutput=2, minInput=-1, minOutput=0.8, parameterName=SpeedFactor}
   volumeModulationParams {default=0, maxOutput=0, minInput=-1, minOutput=1, parameterName=SpeedFactor}
  }
 }
 {id=WavePlayer_1, properties {bLooping=true, soundWaveAssetPtr=/Ascender/Gameplay/Ascender/Sounds/wav/Ascender_Descend_Loop}}
 {id=WavePlayer_2, properties {bLooping=true, soundWaveAssetPtr=/Ascender/Gameplay/Ascender/Sounds/wav/Ascender_Ascend_Loop}}
 {id=WavePlayer_3, properties {bLooping=true, soundWaveAssetPtr=&Ascender_Descend_Loop}}
 {id=WavePlayer_4, properties {bLooping=true, soundWaveAssetPtr=&Ascender_Ascend_Loop}}
 {id=WavePlayer_6, properties {bLooping=true, soundWaveAssetPtr=&Ascender_Descend_Loop}}
 {id=WavePlayer_7, properties {bLooping=true, soundWaveAssetPtr=&Ascender_Ascend_Loop}}
]
properties {
 attenuationSettings=/Game/Sounds/Attenuation/Fort_Ascender_Att
 maxDistance=10000
 soundClassObject=&Fort_Gadgets_Items
 volumeMultiplier=1.12
}
