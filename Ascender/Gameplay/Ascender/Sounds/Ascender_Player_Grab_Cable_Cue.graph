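# Sound cue for a player grabbing an Ascender cable, using zipline attachment attenuation with low-pass and occlusion overrides.
# An IsLocal branch gives the local player a random grab clip layered with a modulated whoosh.
# Otherwise, an IsEnemy branch plays only the random grab, with non-enemies much quieter than enemies or an unset parameter.
# Concurrency allows one instance per owner and prevents new ones while it plays.
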SoundCue /Ascender/Gameplay/Ascender/Sounds/Ascender_Player_Grab_Cable_Cue
root=Branch_0
nodes [
 {id=Branch_0, children [Mixer_5,Branch_1,Branch_1], properties {boolParameterName=IsLocal}}
 {id=Branch_1, children [Mixer_1,Mixer_2,Mixer_1], properties {boolParameterName=IsEnemy}}
 {id=Mixer_0, children [Random_0,Modulator_2], properties {inputVolume [1,1]}}
 {id=Mixer_1, children [Random_1], properties {inputVolume [1]}}
 {id=Mixer_2, children [Random_2], properties {inputVolume [0.25]}}
 {id=Mixer_5, children [Mixer_0], properties {inputVolume [0.65]}}
 {id=Modulator_2, children [WavePlayer_4]}
 {id=Random_0, children [WavePlayer_1,WavePlayer_2,WavePlayer_3], properties {bShouldExcludeFromBranchCulling=true, weights [1,1,1]}}
 {id=Random_1, children [WavePlayer_0,WavePlayer_5,WavePlayer_6], properties {bShouldExcludeFromBranchCulling=true, weights [1,1,1]}}
 {id=Random_2, children [WavePlayer_8, WavePlayer_9, WavePlayer_10], properties {bShouldExcludeFromBranchCulling=true, weights [1,1,1]}}
 {id=WavePlayer_0, properties {soundWaveAssetPtr=/Ascender/Gameplay/Ascender/Sounds/wav/Ascender_Ascender_Grab_Descend_01}}
 {id=WavePlayer_1, properties {soundWaveAssetPtr=&Ascender_Ascender_Grab_Descend_01}}
 {id=WavePlayer_10, properties {soundWaveAssetPtr=/Ascender/Gameplay/Ascender/Sounds/wav/Ascender_Ascender_Grab_Descend_03}}
 {id=WavePlayer_2, properties {soundWaveAssetPtr=/Ascender/Gameplay/Ascender/Sounds/wav/Ascender_Ascender_Grab_Descend_02}}
 {id=WavePlayer_3, properties {soundWaveAssetPtr=&Ascender_Ascender_Grab_Descend_03}}
 {id=WavePlayer_4, properties {soundWaveAssetPtr=/Ascender/Gameplay/Ascender/Sounds/wav/Ascender_Ascender_Grap_Whoosh}}
 {id=WavePlayer_5, properties {soundWaveAssetPtr=&Ascender_Ascender_Grab_Descend_02}}
 {id=WavePlayer_6, properties {soundWaveAssetPtr=&Ascender_Ascender_Grab_Descend_03}}
 {id=WavePlayer_8, properties {soundWaveAssetPtr=&Ascender_Ascender_Grab_Descend_01}}
 {id=WavePlayer_9, properties {soundWaveAssetPtr=&Ascender_Ascender_Grab_Descend_02}}
]
properties {
 attenuationSettings=/Game/Sounds/Attenuation/Fort_Zipline_Attachment_Att
 bOverrideAttenuation=true
 bOverrideConcurrency=true
 maxDistance=3900
 soundClassObject=&Fort_Gadgets_Items
 volumeMultiplier=0.7
 attenuationOverrides {
  bAttenuateWithLPF=true
  bEnableOcclusion=true
  falloffDistance=3500
  lPFFrequencyAtMax=1500
  lPFRadiusMax=2000
  lPFRadiusMin=400
  occlusionLowPassFilterFrequency=5000
  occlusionTraceChannel=ECC_GameTraceChannel11
  reverbDistanceMax=2000
 }
 concurrencyOverrides {bLimitToOwner=true, maxCount=1, resolutionRule=PreventNew}
}
