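# Sound cue for a player hopping off an Ascender, playing a whoosh and click with zipline attachment attenuation and limited concurrency.
# Branches on IsLocal, then IsEnemy, choose the mix: local gets a softened whoosh plus click, enemy or unset gets both at full volume, non-enemy gets only a quiet whoosh.
# Each path randomly picks among the same three whooshes, with random nodes excluded from branch culling.
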SoundCue /Ascender/Gameplay/Ascender/Sounds/Ascender_Player_HopOff_Whoosh_Cue
root=Branch_0
nodes [
 {id=Branch_0, children [Mixer_5,Branch_1,Branch_1], properties {boolParameterName=IsLocal}}
 {id=Branch_1, children [Mixer_1,Mixer_2,Mixer_1], properties {boolParameterName=IsEnemy}}
 {id=Mixer_1, children [Modulator_1,Modulator_4], properties {inputVolume [1,1]}}
 {id=Mixer_2, children [Modulator_2], properties {inputVolume [0.25]}}
 {id=Mixer_5, children [Modulator_0,Modulator_3], properties {inputVolume [0.65,1]}}
 {id=Modulator_0, children [Random_1], properties {pitchMax=1, pitchMin=1}}
 {id=Modulator_1, children [Random_0], properties {pitchMax=1, pitchMin=1}}
 {id=Modulator_2, children [Random_2], properties {pitchMax=1, pitchMin=1}}
 {id=Modulator_3, children [WavePlayer_9]}
 {id=Modulator_4, children [WavePlayer_10]}
 {id=Random_0, children [WavePlayer_0,WavePlayer_1,WavePlayer_2], properties {bShouldExcludeFromBranchCulling=true, weights [1,1,1]}}
 {id=Random_1, children [WavePlayer_6,WavePlayer_7,WavePlayer_8], properties {bShouldExcludeFromBranchCulling=true, weights [1,1,1]}}
 {id=Random_2, children [WavePlayer_3,WavePlayer_4,WavePlayer_5], properties {bShouldExcludeFromBranchCulling=true, weights [1,1,1]}}
 {id=WavePlayer_0, properties {soundWaveAssetPtr=/Ascender/Gameplay/Ascender/Sounds/wav/Ascender_Ascender_Player_HopOff_Whoosh_04}}
 {id=WavePlayer_1, properties {soundWaveAssetPtr=/Ascender/Gameplay/Ascender/Sounds/wav/Ascender_Ascender_Player_HopOff_Whoosh_06}}
 {id=WavePlayer_10, properties {soundWaveAssetPtr=/Ascender/Gameplay/Ascender/Sounds/wav/Ascender_Ascender_Player_HopOff_Click_01}}
 {id=WavePlayer_2, properties {soundWaveAssetPtr=/Ascender/Gameplay/Ascender/Sounds/wav/Ascender_Ascender_Player_HopOff_Whoosh_05}}
 {id=WavePlayer_3, properties {soundWaveAssetPtr=&Ascender_Ascender_Player_HopOff_Whoosh_04}}
 {id=WavePlayer_4, properties {soundWaveAssetPtr=&Ascender_Ascender_Player_HopOff_Whoosh_06}}
 {id=WavePlayer_5, properties {soundWaveAssetPtr=&Ascender_Ascender_Player_HopOff_Whoosh_05}}
 {id=WavePlayer_6, properties {soundWaveAssetPtr=&Ascender_Ascender_Player_HopOff_Whoosh_04}}
 {id=WavePlayer_7, properties {soundWaveAssetPtr=&Ascender_Ascender_Player_HopOff_Whoosh_06}}
 {id=WavePlayer_8, properties {soundWaveAssetPtr=&Ascender_Ascender_Player_HopOff_Whoosh_05}}
 {id=WavePlayer_9, properties {soundWaveAssetPtr=&Ascender_Ascender_Player_HopOff_Click_01}}
]
properties {
 attenuationSettings=/Game/Sounds/Attenuation/Fort_Zipline_Attachment_Att
 bOverrideConcurrency=true
 maxDistance=10000
 soundClassObject=&Fort_Gadgets
 volumeMultiplier=1
 attenuationOverrides {falloffDistance=1536, stereoSpread=0, attenuationShapeExtents {x=512}}
 concurrencyOverrides {maxCount=2}
}
