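# Sound cue for the looping wind and cable sound heard as a player rides an ascender, using ascender attenuation.
# An IsLocal branch gives the local player layered descend wind and cable loops.
# Others hear only the cable loop, which a nested IsEnemy branch plays louder for enemies or unset values and much quieter for non-enemies.
# Each path runs through a modulator and an enveloper that fades volume in over roughly a second.
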SoundCue /Ascender/Gameplay/Ascender/Sounds/Ascender_Player_Wind_Loop_Cue
root=Branch_0
nodes [
 {id=Branch_0, children [Mixer_5,Branch_1,Branch_1], properties {boolParameterName=IsLocal}}
 {id=Branch_1, children [Mixer_1,Mixer_2,Mixer_1], properties {boolParameterName=IsEnemy}}
 {
  id=Enveloper_0
  children [Modulator_0]
  properties {
   pitchCurve {editorCurveData {keys [{time=0}]}}
   volumeCurve {editorCurveData {keys [
      <at 0 0 RCIM_Cubic>
      {arriveTangent=1.5256, interpMode=RCIM_Cubic, leaveTangent=1.5256, tangentMode=RCTM_User, time=0.6, value=0.3959}
      <at 1.3 1>
     ]}}
  }
 }
 {
  id=Enveloper_1
  children [Modulator_1]
  properties {
   pitchCurve {editorCurveData {keys [{time=0}]}}
   volumeCurve {editorCurveData {keys [
      <at 0 0 RCIM_Cubic>
      {arriveTangent=1.5256, interpMode=RCIM_Cubic, leaveTangent=1.5256, tangentMode=RCTM_User, time=0.6, value=0.3959}
      <at 1.3 1>
     ]}}
  }
 }
 {
  id=Enveloper_2
  children [Modulator_2]
  properties {
   pitchCurve {editorCurveData {keys [{time=0}]}}
   volumeCurve {editorCurveData {keys [
      <at 0 0 RCIM_Cubic>
      {arriveTangent=1.5256, interpMode=RCIM_Cubic, leaveTangent=1.5256, tangentMode=RCTM_User, time=0.6, value=0.3959}
      <at 1.3 1>
     ]}}
  }
 }
 {id=Mixer_0, children [WavePlayer_0,WavePlayer_1], properties {inputVolume [1,1]}}
 {id=Mixer_1, children [Enveloper_1], properties {inputVolume [1.25]}}
 {id=Mixer_2, children [Enveloper_2], properties {inputVolume [0.25]}}
 {id=Mixer_5, children [Enveloper_0], properties {inputVolume [0.65]}}
 {id=Modulator_0, children [Mixer_0]}
 {id=Modulator_1, children [WavePlayer_3]}
 {id=Modulator_2, children [WavePlayer_4]}
 {id=WavePlayer_0, properties {bLooping=true, soundWaveAssetPtr=/Ascender/Gameplay/Ascender/Sounds/wav/Ascender_Ascender_Descend_Wind_Loop}}
 {id=WavePlayer_1, properties {bLooping=true, soundWaveAssetPtr=/Ascender/Gameplay/Ascender/Sounds/wav/Ascender_Ascender_Descent_Cable_Loop}}
 {id=WavePlayer_3, properties {bLooping=true, soundWaveAssetPtr=&Ascender_Ascender_Descent_Cable_Loop}}
 {id=WavePlayer_4, properties {bLooping=true, soundWaveAssetPtr=&Ascender_Ascender_Descent_Cable_Loop}}
]
properties {
 attenuationSettings=/Game/Sounds/Attenuation/Fort_Ascender_Att
 maxDistance=10000
 soundClassObject=&Fort_Gadgets_Items
 volumeMultiplier=1.008
}
