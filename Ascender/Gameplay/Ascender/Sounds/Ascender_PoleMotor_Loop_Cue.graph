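# Sound cue for the Ascender gadget's pole motor loop, built by layering two motor recordings.
# A mixer blends a looping descend-motor wave pitched up with a quieter ascend-motor wave pitched far down, each through a fixed-pitch modulator.
# Only the descend layer's wave player sets looping.
# It uses the gadgets sound class, with overrides for logarithmic, occlusion-filtered attenuation and a low concurrency cap.
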SoundCue /Ascender/Gameplay/Ascender/Sounds/Ascender_PoleMotor_Loop_Cue
root=Mixer_0
nodes [
 {id=Mixer_0, children [Modulator_1,Modulator_0], properties {inputVolume [0.5,0.35]}}
 {id=Modulator_0, children [WavePlayer_1], properties {pitchMax=0.25, pitchMin=0.25}}
 {id=Modulator_1, children [WavePlayer_2], properties {pitchMax=2, pitchMin=2}}
 {id=WavePlayer_1, properties {soundWaveAssetPtr=/Ascender/Gameplay/Ascender/Sounds/wav/Ascender_Ascend_Motor_Loop}}
 {
  id=WavePlayer_2
  properties {bLooping=true, soundWaveAssetPtr=/Ascender/Gameplay/Ascender/Sounds/wav/Ascender_Ascender_PoleMotor_Descend_Lop}
 }
]
properties {
 bOverrideAttenuation=true
 bOverrideConcurrency=true
 maxDistance=1850
 soundClassObject=&Fort_Gadgets
 volumeMultiplier=1
 attenuationOverrides {
  bEnableOcclusion=true
  distanceAlgorithm=Logarithmic
  falloffDistance=1500
  occlusionLowPassFilterFrequency=950
  occlusionTraceChannel=ECC_GameTraceChannel11
  stereoSpread=0
  attenuationShapeExtents {x=350}
 }
 concurrencyOverrides {maxCount=2}
}
